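# Unlit, two-sided translucent Niagara sprite and mesh material for Anodized Metal pickaxe effects.
# A channel-packed texture supplies masks that blend two colors, with static switches for material colors and RGBA masking.
# Panning noise, optionally shaped by a rotatable mirrored gradient, drives smoothstep erosion with a colored edge, controlled by a dynamic parameter.
# A panning normal map also computes particle-controlled UV distortion, though the main texture sample shows no UV input.
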Material /BRCosmetics/Weapons/FORT_Melee/Pickaxe_AnodizedMetal/FX/M_RGB_Sprite
graph {
 properties {
  bAutomaticallySetUsageInEditor=false
  bUseTranslucencyVertexFog=false
  bUsedWithNiagaraMeshParticles=true
  bUsedWithNiagaraSprites=true
  blendMode=BLEND_Translucent
  shadingModel=MSM_Unlit
  twoSided=true
  usedShadingModels=Unlit
 }
 nodes [
  n1 = TextureCoordinate()
  n2 = VectorParameter() {parameterName=Distortion UV Controls, channelNames {a=Speed V, b=Speed U, g=Tile V, r=Tile U}, defaultValue=<rgba 1 1 0 0.3>}
  n3 = AppendVector(A=@n2.Tile U, B=@n2.Tile U)
  n4 = TextureSampleParameter2D(UVs=Panner(Coordinate=Multiply(A=@n1, B=@n3), Speed=AppendVector(A=@n2.Speed U, B=@n2.Speed U))) {
   parameterName=Distortion Tex
   samplerType=SAMPLERTYPE_Normal
   texture=/BRCosmetics/Gadgets/Assets/VinderTech_GliderChute/Glider_Teriyaki_Fish_Toon/FX/Texture/T_SwirlNoise_N
  }
  n5 = ScalarParameter() {parameterName=Distortion Intensity Debug}
  n6 = DynamicParameter() {defaultValue=<rgba 0 0 0 0>, paramNames [Dissolve, Distortion Intensity, Param3, Param4]}
  Add(A=TextureCoordinate(), B=Multiply(A=ComponentMask(None=@n4.RGB) {g=true, r=true}, B=Add(A=@n5, B=@n6.Distortion Intensity)))
  Multiply(A=StaticSwitchParameter(True=NamedRerouteUsage(), False=NamedRerouteUsage()) {parameterName=Use RGBA Tex?}, B=NamedRerouteUsage())
  n7 = TextureCoordinate()
  n8 = VectorParameter() {parameterName=Dissolve UV Controls, channelNames {a=Speed V, b=Speed U, g=Tile V, r=Tile U}, defaultValue=<rgba 1 1 0 0.3>}
  n9 = AppendVector(A=@n8.Tile U, B=@n8.Tile U)
  n10 = TextureSampleParameter2D(UVs=Panner(Coordinate=Multiply(A=@n7, B=@n9), Speed=AppendVector(A=@n8.Speed U, B=@n8.Speed U))) {
   parameterName=Dissolve Tex
   samplerType=SAMPLERTYPE_LinearGrayscale
   texture=/BRCosmetics/Gadgets/Assets/VinderTech_GliderChute/Glider_GateHound/FX/T_Mojo_Fire_Noise
  }
  n11 = Saturate(None=@n10.G)
  n12 = Add(
   A=ScalarParameter() {parameterName=Dissolve_Debug}
   B=DynamicParameter() {defaultValue=<rgba 0 0 0 0>, paramNames [Dissolve, Distortion Intensity, Param3, Param4]}.Dissolve
  )
  n13 = Subtract(A=NamedRerouteUsage(), B=@n12) {constB=0}
  n14 = MaterialFunctionCall!(
   Texture=LinearInterpolate(A=@n11, B=Multiply(A=@n13, B=@n11), Alpha=ScalarParameter() {parameterName=UseAngledGradient?})
   Erosion=@n12
   Softness=ScalarParameter() {defaultValue=0.1, parameterName=Dissolve Softness}
  ) {materialFunction=/Game/Effects/Niagara/MasterMaterials/Functions/MF_Erosion_Smoothstep_01}
  n15 = Ceil(None=@n14.Result)
  n16 = ScalarParameter() {defaultValue=0.2, parameterName=Dissolve Edge Thickness}
  n17 = OneMinus(None=Ceil(None=Subtract(A=@n14.Result, B=@n16)))
  NamedRerouteDeclaration(None=Saturate(None=Multiply(A=@n15, B=@n17))) {name=Dissolve Edge, nodeColor=<rgba 0.6875 0 1 1>}
  NamedRerouteDeclaration(None=@n14.Result) {name=Dissolve Result, nodeColor=<rgba 0 1 0.3672 1>}
  n18 = TextureSampleParameter2D() {
   group=Textures
   parameterName=Alpha Tex
   samplerSource=SSM_Clamp_WorldGroupSettings
   samplerType=SAMPLERTYPE_LinearColor
   sortPriority=0
   texture=&/T_Sakura_Flower_RGBA
  }
  n19 = VectorParameter() {parameterName=Channel Selector, defaultValue=<rgba 1 0 0 0>}
  NamedRerouteDeclaration(None=DotProduct(
    A=@n18.RGBA
    B=MaterialFunctionCall(X=@n19.R, Y=@n19.R, Z=@n19.R, A=@n19.R) {materialFunction=&MakeFloat4}.Result
   )) {name=AlphaTex RGBA, nodeColor=<rgba 0 0.7656 1 1>}
  n20 = TextureCoordinate()
  n21 = VectorParameter() {parameterName=GradientRotator Controls, channelNames {a="", b=Angle, g=Rotation Y, r=Rotation X}, defaultValue=<rgba 0.5 0.5 0 0>}
  n22 = MaterialFunctionCall(X=@n21.Rotation X, Y=@n21.Rotation X) {materialFunction=&MakeFloat2}
  n23 = Subtract(A=ComponentMask(None=MaterialFunctionCall(UVs=@n20, Rotation Center=@n22.Result, "Rotation Angle (0-1)"=@n21.Angle) {materialFunction=&CustomRotator}.Rotated Values) {r=true}) {constB=0.5}
  NamedRerouteDeclaration(None=Multiply(A=Abs(None=@n23)) {constB=2}) {name=AngledGradient, nodeColor=<rgba 0 1 0.7188 1>}
  NamedRerouteDeclaration(None=Multiply(A=VectorParameter() {parameterName=Dissolve Edge Color, defaultValue=<rgba 20 8.1684 0 1>}.RGB, B=NamedRerouteUsage())) {name=Dissolve Colored Edge, nodeColor=<rgba 1 0.0469 0 1>}
  NamedRerouteDeclaration(None=@n18.R) {name=MainTex R, nodeColor=<rgba 0 0.2734 1 1>}
  NamedRerouteDeclaration(None=@n18.G) {name=MainTex G, nodeColor=<rgba 0 0.2734 1 1>}
  NamedRerouteDeclaration(None=@n18.B) {name=MainTex B, nodeColor=<rgba 0 0.2734 1 1>}
  n24 = Constant3Vector()
  n25 = LinearInterpolate(
   A=VectorParameter() {group=Colors, parameterName=Color_A, sortPriority=0, defaultValue=<rgba 0.4621 0.2623 0.4179 1>}.RGB
   B=VectorParameter() {group=Colors, parameterName=Color_B, sortPriority=1, defaultValue=<rgba 0.2961 0.1441 0.3663 1>}.RGB
   Alpha=LinearInterpolate(A=NamedRerouteUsage(), B=NamedRerouteUsage(), Alpha=ScalarParameter() {parameterName=Mask R or G})
  )
  n26 = NamedRerouteUsage()
  NamedRerouteDeclaration(None=Multiply(A=LinearInterpolate(A=@n24, B=@n25, Alpha=@n26), B=NamedRerouteUsage())) {name=Mat Colors, nodeColor=<rgba 0 1 0.7656 1>}
  n27 = ParticleColor()
  NamedRerouteDeclaration(None=@n27.RGB) {name=Particle RGB, nodeColor=<rgba 1 0.4453 0 1>}
  NamedRerouteDeclaration(None=@n27.A) {name=Particle Alpha, nodeColor=<rgba 0.125 0 1 1>}
  n28 = StaticSwitchParameter(True=NamedRerouteUsage(), False=NamedRerouteUsage()) {parameterName=UseMatColors?}
  NamedRerouteDeclaration(None=MaterialFunctionCall(Emissive=@n28) {materialFunction=&MF_FX_EmissiveAdjustments}.Result) {name=Emissive, nodeColor=<rgba 0.1953 1 0 1>}
  NamedRerouteUsage()
  StaticSwitchParameter(True=Add(A=@n28, B=NamedRerouteUsage()), False=@n28) {parameterName=UseDissolve?}
 ]
}
