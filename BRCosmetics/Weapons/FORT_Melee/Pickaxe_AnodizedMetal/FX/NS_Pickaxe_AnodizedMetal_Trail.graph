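# Niagara swing trail for the AnodizedMetal pickaxe cosmetic, with six looks selected by User.Variant.
# Trails and TrailGalaxy build socket-driven anim-trail ribbons, with Trails layering glow, smoke, tip and highlight colors.
# Symbols spawns short-lived sprites and ribbons by socket movement, spreading them with drag and curl noise.
# System time-of-day color tints all three emitters, TrailFade dims both trails, and Trails also tapers when the system deactivates.
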NiagaraSystem /BRCosmetics/Weapons/FORT_Melee/Pickaxe_AnodizedMetal/FX/NS_Pickaxe_AnodizedMetal_Trail
dependencies {
 usedDataInterfaces [
  &NiagaraDataInterfaceCurve
  &NiagaraDataInterfaceColorCurve
  &NiagaraDataInterfaceArrayFloat
  &NiagaraDataInterfaceSocketReader
 ]
 usedDynamicInputs [&SetBoolByIntComparison]
 usedModules [
  &self:SystemSpawnScript.SystemScriptSource.SystemScriptGraph.NiagaraNodeAssignment_6.SetVariables_#1
  &self:SystemSpawnScript.SystemScriptSource.SystemScriptGraph.NiagaraNodeAssignment_10.SetVariables_#2
  &self:SystemSpawnScript.SystemScriptSource.SystemScriptGraph.NiagaraNodeAssignment_16.SetVariables_#3
  &self:SystemSpawnScript.SystemScriptSource.SystemScriptGraph.NiagaraNodeAssignment_12.SetVariables_#4
  &self:SystemSpawnScript.SystemScriptSource.SystemScriptGraph.NiagaraNodeAssignment_13.SetVariables_#5
  &self:SystemSpawnScript.SystemScriptSource.SystemScriptGraph.NiagaraNodeAssignment_15.SetVariables_#6
  &self:SystemSpawnScript.SystemScriptSource.SystemScriptGraph.NiagaraNodeAssignment_14.SetVariables_#7
  &SystemState
  &self:SystemSpawnScript.SystemScriptSource.SystemScriptGraph.NiagaraNodeAssignment_4.SetVariables_#8
  &self:SystemSpawnScript.SystemScriptSource.SystemScriptGraph.NiagaraNodeAssignment_5.SetVariables_#9
  &self:SystemSpawnScript.SystemScriptSource.SystemScriptGraph.NiagaraNodeAssignment_9.SetVariables_#10
  &self:SystemSpawnScript.SystemScriptSource.SystemScriptGraph.NiagaraNodeAssignment_7.SetVariables_#11
  &self:ScratchModule_03
 ]
}
emitters {
 Symbols {
  emitterUpdateScript [
   &EmitterState()
   &SpawnRate(
    SpawnRate=&MakeCustomFloatFromBool(Bool=$User.Debug, True Float=10, False Float=0)
    Use Spawn Probability=false
   )
   &SpawnPerUnit(
    Max Movement Threshold=4000
    Module.Use Max Movement Threshold=true
    Module.Use Movement Tolerance=false
    Spawn Probability=1
    Spawn Spacing:NiagaraFloat=&SelectFloatFromArray(
     Direct Array Index=&MakeIntFromFloat(Float=$User.Variant)
     Float Selection Array=DataInterface() {propertyValues {FloatData [0,25,25,25,25,12]}}
    )
    Use Spawn Probability=true
   )
  ]
  particleSpawnScript [
   &InitializeParticle(
    Color=&MergeHelper_ColorSwitch(
     0=<rgba 0 0 0 0>
     1=&ColorFromCurve(
      ColorCurve=DataInterface() {propertyValues {
        AlphaCurve {keys [<at 0 1 RCIM_Cubic>]}
        BlueCurve {keys [<at 0.4274 0.9047 RCIM_Cubic>,<at 0.4699 0.6751 RCIM_Cubic>]}
        GreenCurve {keys [<at 0.4274 0.8796 RCIM_Cubic>,<at 0.4699 0.6696 RCIM_Cubic>]}
        RedCurve {keys [<at 0.4274 1 RCIM_Cubic>,<at 0.4699 1 RCIM_Cubic>]}
       }}
      CurveIndex=&RandomRangeFloat(Minimum=0, Maximum=1)
     )
     2=&ColorFromCurve(
      ColorCurve=DataInterface() {propertyValues {
        AlphaCurve {keys [<at 0 1 RCIM_Cubic>]}
        BlueCurve {keys [<at 0.4274 0.1 RCIM_Cubic>,<at 0.4699 0.1 RCIM_Cubic>]}
        GreenCurve {keys [<at 0.4274 0.6609 RCIM_Cubic>,<at 0.4699 0.4283 RCIM_Cubic>]}
        RedCurve {keys [<at 0.4274 1 RCIM_Cubic>,<at 0.4699 1 RCIM_Cubic>]}
       }}
      CurveIndex=&RandomRangeFloat(Minimum=0, Maximum=1)
     )
     3=&ColorFromCurve(
      ColorCurve=DataInterface() {propertyValues {
        AlphaCurve {keys [<at 0 1 RCIM_Cubic>]}
        BlueCurve {keys [<at 0.4274 1 RCIM_Cubic>,<at 0.4699 1 RCIM_Cubic>]}
        GreenCurve {keys [<at 0.4274 0.9469 RCIM_Cubic>,<at 0.4699 0.6167 RCIM_Cubic>]}
        RedCurve {keys [<at 0.4274 0.5208 RCIM_Cubic>,<at 0.4699 0.4375 RCIM_Cubic>]}
       }}
      CurveIndex=&RandomRangeFloat(Minimum=0, Maximum=1)
     )
     4=&ColorFromCurve(
      ColorCurve=DataInterface() {propertyValues {
        AlphaCurve {keys [<at 0 1 RCIM_Cubic>]}
        BlueCurve {keys [<at 0.4274 1 RCIM_Cubic>,<at 0.4699 1 RCIM_Cubic>]}
        GreenCurve {keys [<at 0.4274 0.7649 RCIM_Cubic>,<at 0.4699 0.3664 RCIM_Cubic>]}
        RedCurve {keys [<at 0.4274 0.3886 RCIM_Cubic>,<at 0.4699 0.7831 RCIM_Cubic>]}
       }}
      CurveIndex=&RandomRangeFloat(Minimum=0, Maximum=1)
     )
     5=&ColorFromCurve(
      ColorCurve=DataInterface() {propertyValues {
        AlphaCurve {keys [<at 0 1 RCIM_Cubic>]}
        BlueCurve {keys [<at 0.4274 1 RCIM_Cubic>,<at 0.4699 1 RCIM_Cubic>]}
        GreenCurve {keys [<at 0.4274 0 RCIM_Cubic>,<at 0.4699 0.282 RCIM_Cubic>]}
        RedCurve {keys [<at 0.4274 0.6698 RCIM_Cubic>,<at 0.4699 0 RCIM_Cubic>]}
       }}
      CurveIndex=&RandomRangeFloat(Minimum=0, Maximum=1)
     )
     Index=&MakeIntFromFloat(Float=$User.Variant, Conversion Mode=::NewEnumerator1)
     Num=::NewEnumerator4
    )
    Lifetime=&MergeHelper_FloatSwitch(
     0=0
     1=&RandomRangeFloat(Minimum=0.35, Maximum=0.45)
     2=&RandomRangeFloat(Minimum=0.3, Maximum=0.4)
     3=&RandomRangeFloat(Minimum=0.35, Maximum=0.45)
     4=&RandomRangeFloat(Minimum=0.25, Maximum=0.35)
     5=0.2
     Index=&MakeIntFromFloat(Float=$User.Variant, Conversion Mode=::NewEnumerator1)
     Num=::NewEnumerator4
    )
    Ribbon Width=1
    Ribbon Width Mode=::NewEnumerator1
    Sprite Rotation Angle=&MergeHelper_FloatSwitch(
     0=0
     1=&RandomRangeFloat(Minimum=0, Maximum=1)
     2=0
     3=&RandomRangeFloat(Minimum=0, Maximum=1)
     4=0
     5=0
     Index=&MakeIntFromFloat(Float=$User.Variant, Conversion Mode=::NewEnumerator1)
     Num=::NewEnumerator4
    )
    Sprite Rotation Mode=::NewEnumerator1
    Sprite Size Mode=::NewEnumerator3
    Uniform Sprite Size=&MergeHelper_FloatSwitch(
     0=0
     1=&RandomRangeFloat(Minimum=10, Maximum=14)
     2=&RandomRangeFloat(Minimum=12, Maximum=16)
     3=&RandomRangeFloat(Minimum=9, Maximum=13)
     4=&RandomRangeFloat(Minimum=7, Maximum=10)
     5=&RandomRangeFloat(Minimum=15, Maximum=20)
     Index=&MakeIntFromFloat(Float=$User.Variant, Conversion Mode=::NewEnumerator1)
     Num=::NewEnumerator4
    )
    UsePositionOffset=true
    Write Material Random=true
   )
   &AddVelocity(Velocity=&MakeCustomVectorFromBool(Bool=$User.Debug, True Vector=<350 0 0>, False Vector=<0 0 0>))
   &SocketLocation(
    Interpolate Socket Sampling=true
    SamplePosOffset=true
    Sampled Position Offset=&MergeHelper_VectorSwitch(
     0=<0 25 0>
     1=<0 25 0>
     2=<0 25 0>
     3=<0 25 0>
     4=<0 25 0>
     5=<0 0 0>
     Index=&MakeIntFromFloat(Float=$User.Variant)
     Num=::NewEnumerator4
    )
    Socket Index=1
    Socket Reader=$System.Socket Reader
    Socket Sampling Mode=::NewEnumerator3
    VelocitySpeedLimit=false
   )
   &AddVelocity(
    Rotation Coordinate Space=::Simulation
    Velocity=&RotateVectorByQuaternion(Vector To Rotate=<0 0 1>, Quaternion=$Output.SocketLocation.SampledRotation)
    Velocity Speed Scale=&MakeCustomFloatFromBool(Bool=$User.Debug, True Float=0, False Float=&RandomRangeFloat(Minimum=75, Maximum=150))
   ) {module=AddVelocity001}
   /SetVariables(Particles.SubImageIndexAlt:NiagaraFloat=&RandomRangeFloat(Minimum=0, Maximum=36)) {isSetParameters=true}
  ]
  particleUpdateScript [
   &ParticleState()
   &SocketLocation(
    Interpolate Socket Sampling=true
    Position Sampling=::NewEnumerator1
    SamplePosOffset=false
    Socket Index=1
    Socket Reader=$System.Socket Reader
    Socket Sampling Mode=::NewEnumerator3
    VelocitySpeedLimit=false
   ) {module=SocketLocation001}
   &ScaleSpriteSize(Uniform Curve Sprite Scale=DataInterface() {propertyValues {Curve {keys [
        {arriveTangent=8.2127, interpMode=RCIM_Cubic, leaveTangent=8.2127, tangentMode=RCTM_User, time=0, value=0}
        {
         arriveTangent=0.0908
         arriveTangentWeight=0.232
         interpMode=RCIM_Cubic
         leaveTangent=0.0908
         leaveTangentWeight=2.2538
         tangentMode=RCTM_User
         time=0.1
         value=1
        }
        {arriveTangent=-0.042, interpMode=RCIM_Cubic, leaveTangent=-0.042, tangentMode=RCTM_User, time=1, value=0}
       ]}}}) {module=ScaleSpriteSize001}
   &Drag(Drag=&RandomRangeFloat(Minimum=6, Maximum=8), Use Linear Drag=true, Use Rotational Drag=true)
   &CurlNoiseForce(
    Noise Strength=&FloatFromCurve(
     FloatCurve=DataInterface() {propertyValues {Curve {keys [
         {arriveTangent=2.7279, interpMode=RCIM_Cubic, leaveTangent=2.7279, tangentMode=RCTM_User, time=0, value=0}
         {interpMode=RCIM_Cubic, tangentMode=RCTM_User, time=1, value=1}
        ]}}}
     Scale Curve=550
    )
    Pan Noise=false
   )
   &ScaleRibbonWidth(Ribbon Width Scale=&FloatFromCurve(FloatCurve=DataInterface() {propertyValues {Curve {keys [{interpMode=RCIM_Cubic, tangentMode=RCTM_User, time=0, value=1}, {interpMode=RCIM_Cubic, tangentMode=RCTM_User, time=1, value=0}]}}}))
   &SolveForcesAndVelocity(Clamp Velocity=false, Limit Acceleration=false)
   &ScaleColor(
    Scale Alpha=&MakeFloatFromLinearColor(LinearColor=$System.TODColorMultiplier, Channel:ENiagaraLinearColor_Channels=::NewEnumerator3)
    Scale RGB=&MakeVectorFromLinearColorRGB(Color=$System.TODColorMultiplier)
    ScaleA=true
    ScaleRGB=true
   )
  ]
  rendererClasses [&NiagaraSpriteRendererProperties, &NiagaraRibbonRendererProperties]
  renderers [
   {rendererClass=&NiagaraSpriteRendererProperties}
   {rendererClass=&NiagaraSpriteRendererProperties, rendererIndex=1}
   {rendererClass=&NiagaraSpriteRendererProperties, rendererIndex=2}
   {rendererClass=&NiagaraSpriteRendererProperties, rendererIndex=3}
   {rendererClass=&NiagaraRibbonRendererProperties, rendererIndex=4}
   {rendererClass=&NiagaraSpriteRendererProperties, rendererIndex=5}
  ]
 }
 TrailGalaxy {
  emitterUpdateScript [
   &EmitterState()
   &FN_SpawnAnimTrail(Spawn Interval=0)
  ]
  particleSpawnScript [
   &FN_InitializeAnimTrailBoneIndecies()
   /Game/Effects/Niagara/Modules/AnimTrail/FN_InitializeAnimTrail/FN_InitializeAnimTrail_SocketReader/FN_InitializeAnimTrail_SocketReader(Color:LinearColor=<rgba 2 2 2 1>, Lifetime=0.4, Socket Reader:NiagaraDataInterfaceSocketReader=$System.Socket Reader)
   &DynamicMaterialParameters(
    Index 0 Param 2=1
    Index 0 Param 3=5
    Index 0 Param 4=0.5
    Index 1 Param 2=0.464
    Index 1 Param 3=0.088
    Module.Index0Float0Write=true
    Module.Index0Float1Write=true
    Module.Index0Float2Write=true
    Module.Index0Float3Write=true
    Module.Index1Float0Write=true
    Module.Index1Float1Write=true
    Module.Index1Float2Write=true
    Module.Index1Float3Write=true
    Write Parameter Index 1=::NewEnumerator1
   )
  ]
  particleUpdateScript [
   /Game/Effects/Niagara/Modules/AnimTrail/FN_UpdateAnimTrail/FN_UpdateAnimTrail_SocketReader/FN_UpdateAnimTrail_SocketReader(Socket Reader:NiagaraDataInterfaceSocketReader=$System.Socket Reader)
   &ParticleState()
   &ScaleColor(
    Scale Alpha=&Multiply_Float(
     A=&MakeFloatFromLinearColor(LinearColor=$System.TODColorMultiplier, Channel:ENiagaraLinearColor_Channels=::NewEnumerator3)
     B=&FloatFromCurve(
      FloatCurve=DataInterface() {propertyValues {Curve {keys [
          {arriveTangent=-1.1204, interpMode=RCIM_Cubic, leaveTangent=-1.1204, tangentMode=RCTM_User, time=0, value=1}
          {interpMode=RCIM_Cubic, tangentMode=RCTM_User, time=1, value=0}
         ]}}}
      Scale Curve=&Multiply_Float(A=&FNTimeOfDayScalar(Night Scale=1, Morning Scale=1, Evening Scale=1), B=$System.TrailFade)
     )
    )
    Scale RGB=&MakeVectorFromLinearColorRGB(Color=$System.TODColorMultiplier)
    ScaleA=true
    ScaleRGB=true
   ) {module=ScaleColor001}
  ]
  rendererClasses [&NiagaraRibbonRendererProperties]
  renderers [{rendererClass=&NiagaraRibbonRendererProperties}]
 }
 Trails {
  emitterUpdateScript [
   &EmitterState()
   &FN_SpawnAnimTrail_v2(Spawn Interval=0)
  ]
  particleSpawnScript [
   /Game/Effects/Niagara/Modules/AnimTrail/FN_InitializeAnimTrail/FN_InitializeAnimTrailBoneIndecies(Second Bone or Socket Index=1)
   &FN_InitializeAnimTrail_SocketReader(Color:LinearColor=<rgba 1.2 1.2 1.2 5>, Lifetime=0.35, Socket Reader:NiagaraDataInterfaceSocketReader=$System.Socket Reader)
   /SetVariables(Particles.BG Smoke Color:LinearColor=<rgba 1 1 1 2>, Particles.BG Smoke Lifetime=0.35) {isSetParameters=true}
   /SetVariables(Particles.BG Glow Color:LinearColor=<rgba 1.5 1.5 1.5 1>, Particles.BG Glow Lifetime=0.32) {isSetParameters=true}
   /SetVariables(Particles.Tip Trail Color:LinearColor=<rgba 0.8396 0.8584 0.8906 1>, Particles.Tip Trail Lifetime=0.16) {isSetParameters=true}
   /SetVariables(Particles.Highlights Color:LinearColor=&Multiply_LinearColorByFloat(
     Linear Color=&MergeHelper_ColorSwitch(
      0=<rgba 4 0.6353 3.8709 1>
      1=<rgba 1.5 0.3125 0.8271 1>
      2=<rgba 4 4 4 1>
      3=<rgba 10 10 10 1>
      4=<rgba 7 7 7 1>
      5=<rgba 10 10 10 1>
      Index=&MakeIntFromFloat(Float=$User.Variant)
      Num=::NewEnumerator4
     )
     Float=0.9
    )) {isSetParameters=true}
   /SetVariables(Particles.TrailAlphaFade=1) {isSetParameters=true}
  ]
  particleUpdateScript [
   /SetVariables(Particles.TrailAlphaFade:NiagaraFloat=&Multiply_Float(
     A=$Particles.TrailAlphaFade
     B=&MakeCustomFloatFromBool(
      Bool=$System.ScratchModule_03.SystemDeactivatedThisFrame
      True Float=&FloatFromCurve(
       FloatCurve=DataInterface() {propertyValues {Curve {keys [
           {tangentMode=RCTM_User, time=0, value=1}
           {arriveTangent=-0.5312, interpMode=RCIM_Cubic, leaveTangent=-0.5312, tangentMode=RCTM_User, time=0.75, value=1}
           {arriveTangent=-6.1768, interpMode=RCIM_Cubic, leaveTangent=-6.1768, tangentMode=RCTM_User, time=1, value=0}
          ]}}}
       CurveIndex=&ReturnNormalizedExecIndex()
      )
      False Float=1
     )
    )) {isSetParameters=true}
   &FN_UpdateAnimTrail_SocketReader(Socket Reader:NiagaraDataInterfaceSocketReader=$System.Socket Reader)
   &ScaleColor(
    Scale Alpha=&Multiply_Float(
     A=&MakeFloatFromLinearColor(LinearColor=$System.TODColorMultiplier, Channel:ENiagaraLinearColor_Channels=::NewEnumerator3)
     B=&Multiply_Float(
      A=&FloatFromCurve(
       FloatCurve=DataInterface() {propertyValues {Curve {keys [
           {arriveTangent=16.3962, interpMode=RCIM_Cubic, leaveTangent=16.3962, tangentMode=RCTM_User, time=0, value=0}
           <at 0.1 1 RCIM_Cubic>
           {arriveTangent=-0.1576, interpMode=RCIM_Cubic, tangentMode=RCTM_Break, time=1, value=0}
          ]}}}
       Scale Curve=&Multiply_Float(
        A=$Particles.TrailAlphaFade
        B=&MergeHelper_FloatSwitch(
         0=1.7
         1=1.7
         2=1.7
         3=1.7
         4=1.7
         5=1.3
         Index=&MakeIntFromFloat(Float=$User.Variant)
         Num=::NewEnumerator4
        )
       )
      )
      B=$System.TrailFade
     )
    )
    Scale RGB=&MakeVectorFromLinearColorRGB(Color=$System.TODColorMultiplier)
    ScaleA=true
    ScaleRGB=true
   ) {module=ScaleColor002}
   /SetVariables(Particles.Update Highlights Color:LinearColor=&Multiply_LinearColorByFloat(
     Linear Color=&MultiplyLinearColor(
      Linear Color=$System.TODColorMultiplier
      Scale Factor=&MakeVector4_FromLinearColor(LinearColor=$Particles.Initial.Highlights Color)
     )
     Float=&Multiply_Float(
      A=&MakeFloatFromLinearColor(LinearColor=$System.TODColorMultiplier, Channel:ENiagaraLinearColor_Channels=::NewEnumerator3)
      B=&Multiply_Float(
       A=&FloatFromCurve(
        FloatCurve=DataInterface() {propertyValues {Curve {keys [
            {arriveTangent=10.1005, interpMode=RCIM_Cubic, leaveTangent=10.1005, tangentMode=RCTM_User, time=0, value=0}
            <at 0.12 1 RCIM_Cubic>
            {arriveTangent=-0.1576, interpMode=RCIM_Cubic, tangentMode=RCTM_Break, time=1, value=0}
           ]}}}
        Scale Curve=&Multiply_Float(
         A=$Particles.TrailAlphaFade
         B=&MergeHelper_FloatSwitch(
          0=1.15
          1=1.15
          2=1.15
          3=1.15
          4=1.15
          5=0.85
          Index=&MakeIntFromFloat(Float=$User.Variant)
          Num=::NewEnumerator4
         )
        )
       )
       B=$System.TrailFade
      )
     )
    )) {isSetParameters=true}
   &DynamicMaterialParameters(
    Index 0 Param 1=&FloatFromCurve(FloatCurve=DataInterface() {propertyValues {Curve {keys [{arriveTangent=-0.7576, interpMode=RCIM_Cubic, time=0, value=-0.1}, {interpMode=RCIM_Cubic, leaveTangent=-0.4845, time=1, value=-0.5}]}}})
    Module.Index0Float0Write=true
    Module.Index0Float1Write=true
    Module.Index0Float2Write=true
    Module.Index0Float3Write=true
   )
   /SetVariables(
    Particles.Update BG Glow Lifetime:NiagaraFloat=&Divide_Float(A=$Particles.Age, B=$Particles.Initial.BG Glow Lifetime)
    Particles.Update BG Smoke Lifetime:NiagaraFloat=&Divide_Float(A=$Particles.Age, B=$Particles.Initial.BG Smoke Lifetime)
    Particles.Update Tip Trail Lifetime:NiagaraFloat=&Divide_Float(A=$Particles.Age, B=$Particles.Initial.Tip Trail Lifetime)
   ) {isSetParameters=true}
   /SetVariables(
    Particles.Update BG Glow Color:LinearColor=&MakeLinearColorFromVectorAndFloat(
     "Vector (RGB)"=&Multiply_Vector(
      A=&Multiply_Vector(
       A=&MakeVectorFromLinearColorRGB(Color=$Particles.Initial.BG Glow Color)
       B=&MakeVectorFromLinearColorRGB(Color=&MergeHelper_ColorSwitch(
         0=&ColorFromCurve(
          ColorCurve=DataInterface() {propertyValues {
            AlphaCurve {keys []}
            BlueCurve {keys [<at 0 0.053 RCIM_Cubic>,<at 0.5169 0.0397 RCIM_Cubic>]}
            GreenCurve {keys [<at 0 0 RCIM_Cubic>,<at 0.5169 0 RCIM_Cubic>]}
            RedCurve {keys [<at 0 1 RCIM_Cubic>,<at 0.5169 0.1198 RCIM_Cubic>]}
           }}
          CurveIndex=$Particles.Update BG Glow Lifetime
         )
         1=&ColorFromCurve(
          ColorCurve=DataInterface() {propertyValues {
            AlphaCurve {keys []}
            BlueCurve {keys [<at 0 0.6644 RCIM_Cubic>,<at 0.5169 0.1454 RCIM_Cubic>]}
            GreenCurve {keys [<at 0 0.2 RCIM_Cubic>,<at 0.5169 0.0756 RCIM_Cubic>]}
            RedCurve {keys [<at 0 1 RCIM_Cubic>,<at 0.5169 0.3478 RCIM_Cubic>]}
           }}
          CurveIndex=$Particles.Update BG Glow Lifetime
         )
         2=&ColorFromCurve(
          ColorCurve=DataInterface() {propertyValues {
            AlphaCurve {keys []}
            BlueCurve {keys [<at 0 0 RCIM_Cubic>,<at 0.5169 0 RCIM_Cubic>]}
            GreenCurve {keys [<at 0 0.0546 RCIM_Cubic>,<at 0.5169 0.0246 RCIM_Cubic>]}
            RedCurve {keys [<at 0 0.1042 RCIM_Cubic>,<at 0.5169 0.0469 RCIM_Cubic>]}
           }}
          CurveIndex=$Particles.Update BG Glow Lifetime
         )
         3=&ColorFromCurve(
          ColorCurve=DataInterface() {propertyValues {
            AlphaCurve {keys []}
            BlueCurve {keys [<at 0 0.5776 RCIM_Cubic>,<at 0.5169 0.4 RCIM_Cubic>]}
            GreenCurve {keys [<at 0 0.4897 RCIM_Cubic>,<at 0.5169 0.3304 RCIM_Cubic>]}
            RedCurve {keys [<at 0 0.1256 RCIM_Cubic>,<at 0.5169 0.0417 RCIM_Cubic>]}
           }}
          CurveIndex=$Particles.Update BG Glow Lifetime
         )
         4=&ColorFromCurve(
          ColorCurve=DataInterface() {propertyValues {
            AlphaCurve {keys []}
            BlueCurve {keys [<at 0 0.4957 RCIM_Cubic>,<at 0.227 0.506>,<at 0.5169 1.1304 RCIM_Cubic>]}
            GreenCurve {keys [<at 0 0.1909 RCIM_Cubic>,<at 0.227 0.5565>,<at 0.5169 0.2556 RCIM_Cubic>]}
            RedCurve {keys [<at 0 0.1177 RCIM_Cubic>,<at 0.227 0.1742>,<at 0.5169 0.9846 RCIM_Cubic>]}
           }}
          CurveIndex=$Particles.Update BG Glow Lifetime
         )
         5=&ColorFromCurve(
          ColorCurve=DataInterface() {propertyValues {
            AlphaCurve {keys []}
            BlueCurve {keys [<at 0 0.3652 RCIM_Cubic>,<at 0.5169 0.395 RCIM_Cubic>]}
            GreenCurve {keys [<at 0 0.2542 RCIM_Cubic>,<at 0.5169 0 RCIM_Cubic>]}
            RedCurve {keys [<at 0 0.0647 RCIM_Cubic>,<at 0.5169 0.0593 RCIM_Cubic>]}
           }}
          CurveIndex=$Particles.Update BG Glow Lifetime
         )
         Index=&MakeIntFromFloat(Float=$User.Variant)
         Num=::NewEnumerator4
        ))
      )
      B=&MakeVectorFromLinearColorRGB(Color=$System.TODColorMultiplier)
     )
     "Float (Alpha)"=&Multiply_Float(
      A=&MakeFloatFromLinearColor(LinearColor=$System.TODColorMultiplier, Channel:ENiagaraLinearColor_Channels=::NewEnumerator3)
      B=&Multiply_Float(
       A=&FloatFromCurve(
        FloatCurve=DataInterface() {propertyValues {Curve {keys [
            {arriveTangent=10.1005, interpMode=RCIM_Cubic, leaveTangent=10.1005, tangentMode=RCTM_User, time=0, value=0}
            <at 0.12 1 RCIM_Cubic>
            {arriveTangent=-0.1576, interpMode=RCIM_Cubic, tangentMode=RCTM_Break, time=1, value=0}
           ]}}}
        Scale Curve=&Multiply_Float(
         A=$Particles.TrailAlphaFade
         B=&MergeHelper_FloatSwitch(0=1, 1=0.5, 2=3, 3=0.5, 4=0.5, 5=1, Index=&MakeIntFromFloat(Float=$User.Variant), Num=::NewEnumerator4)
        )
       )
       B=$System.TrailFade
      )
     )
    )
    Particles.Update BG Smoke Color:LinearColor=&MakeLinearColorFromVectorAndFloat(
     "Vector (RGB)"=&Multiply_Vector(
      A=&MakeVectorFromLinearColorRGB(Color=$Particles.Initial.BG Smoke Color)
      B=&MakeVectorFromLinearColorRGB(Color=$System.TODColorMultiplier)
     )
     "Float (Alpha)"=&Multiply_Float(
      A=&MakeFloatFromLinearColor(LinearColor=$System.TODColorMultiplier, Channel:ENiagaraLinearColor_Channels=::NewEnumerator3)
      B=&Multiply_Float(
       A=&FloatFromCurve(
        FloatCurve=DataInterface() {propertyValues {Curve {keys [
            {arriveTangent=10.1005, interpMode=RCIM_Cubic, leaveTangent=10.1005, tangentMode=RCTM_User, time=0, value=0}
            <at 0.12 1 RCIM_Cubic>
            {arriveTangent=-0.1576, interpMode=RCIM_Cubic, tangentMode=RCTM_Break, time=1, value=0}
           ]}}}
        Scale Curve=$Particles.TrailAlphaFade
       )
       B=$System.TrailFade
      )
     )
    )
    Particles.Update Tip Trail Color:LinearColor=&MultiplyLinearColor(
     Linear Color=&MultiplyLinearColor(
      Linear Color=&MergeHelper_ColorSwitch(
       0=&ColorFromCurve(
        ColorCurve=DataInterface() {propertyValues {
          AlphaCurve {keys [<at 0 0.7>,<at 0.038 1>,<at 0.129 1>,<at 0.2754 0.3>,<at 0.4961 0.05>,<at 0.7442 0>]}
          BlueCurve {keys [<at 0 0.4948>,<at 0.5969 0.5208>]}
          GreenCurve {keys [<at 0 0.4948>,<at 0.5969 0.5208>]}
          RedCurve {keys [<at 0 0.4948>,<at 0.5969 0.5208>]}
         }}
        CurveIndex=$Particles.Update Tip Trail Lifetime
       )
       1=&ColorFromCurve(
        ColorCurve=DataInterface() {propertyValues {
          AlphaCurve {keys [<at 0 0.7>,<at 0.038 1>,<at 0.129 1>,<at 0.2754 0.3>,<at 0.4961 0.05>,<at 0.7442 0>]}
          BlueCurve {keys [<at 0 0.4948>,<at 0.5969 0.5208>]}
          GreenCurve {keys [<at 0 0.4948>,<at 0.5969 0.5208>]}
          RedCurve {keys [<at 0 0.4948>,<at 0.5969 0.5208>]}
         }}
        CurveIndex=$Particles.Update Tip Trail Lifetime
       )
       2=&ColorFromCurve(
        ColorCurve=DataInterface() {propertyValues {
          AlphaCurve {keys [<at 0 0.7>,<at 0.038 1>,<at 0.129 1>,<at 0.2754 0.3>,<at 0.4961 0.05>,<at 0.7442 0>]}
          BlueCurve {keys [<at 0 0.3>,<at 0.5969 0.1652>]}
          GreenCurve {keys [<at 0 0.6657>,<at 0.5969 0.4406>]}
          RedCurve {keys [<at 0 1>,<at 0.5969 0.8261>]}
         }}
        CurveIndex=$Particles.Update Tip Trail Lifetime
       )
       3=&ColorFromCurve(
        ColorCurve=DataInterface() {propertyValues {
          AlphaCurve {keys [<at 0 0.7>,<at 0.038 1>,<at 0.129 1>,<at 0.2754 0.3>,<at 0.4961 0.05>,<at 0.7442 0>]}
          BlueCurve {keys [<at 0 0.4948>,<at 0.5969 0.5208>]}
          GreenCurve {keys [<at 0 0.4076>,<at 0.5969 0.3922>]}
          RedCurve {keys [<at 0 0.3223>,<at 0.5969 0.3207>]}
         }}
        CurveIndex=$Particles.Update Tip Trail Lifetime
       )
       4=&ColorFromCurve(
        ColorCurve=DataInterface() {propertyValues {
          AlphaCurve {keys [<at 0 0.7>,<at 0.038 1>,<at 0.129 1>,<at 0.2754 0.3>,<at 0.4961 0.05>,<at 0.7442 0>]}
          BlueCurve {keys [<at 0 1>,<at 0.2628 0.9093>,<at 0.5969 2>]}
          GreenCurve {keys [<at 0 0.3852>,<at 0.2628 1>,<at 0.5969 0.4522>]}
          RedCurve {keys [<at 0 0.2375>,<at 0.2628 0.313>,<at 0.5969 1.742>]}
         }}
        CurveIndex=$Particles.Update Tip Trail Lifetime
       )
       5=&ColorFromCurve(
        ColorCurve=DataInterface() {propertyValues {
          AlphaCurve {keys [<at 0 0.7>,<at 0.038 1>,<at 0.129 1>,<at 0.2754 0.3>,<at 0.4961 0.05>,<at 0.7442 0>]}
          BlueCurve {keys [<at 0 0.3652>,<at 0.5969 0.395>]}
          GreenCurve {keys [<at 0 0.2542>,<at 0.5969 0>]}
          RedCurve {keys [<at 0 0.0647>,<at 0.5969 0.0593>]}
         }}
        CurveIndex=$Particles.Update Tip Trail Lifetime
       )
       Index=&MakeIntFromFloat(Float=$User.Variant)
       Num=::NewEnumerator4
      )
      Scale Factor=&MakeVector4_FromLinearColor(LinearColor=$System.TODColorMultiplier)
     )
     Scale Factor=&MakeVector4_FromLinearColor(LinearColor=$Particles.Initial.Tip Trail Color)
    )
   ) {isSetParameters=true}
   /SetVariables(
    Particles.DynPar BG Glow:Vector4f=&MakeVector4(
     X=&FloatFromCurve(
      FloatCurve=DataInterface() {propertyValues {Curve {keys [
          {arriveTangent=-0.3031, interpMode=RCIM_Cubic, leaveTangent=-0.3031, tangentMode=RCTM_User, time=0.3, value=0}
          {arriveTangent=-0.7176, interpMode=RCIM_Cubic, leaveTangent=-0.7176, tangentMode=RCTM_User, time=1, value=-0.5}
         ]}}}
      CurveIndex=$Particles.Update BG Glow Lifetime
     )
     Y=0
     Z=0
     W=0
    )
    Particles.DynPar BG Smoke:Vector4f=&MakeVector4(
     X=&FloatFromCurve(FloatCurve=DataInterface() {propertyValues {Curve {keys [<at 0 3 RCIM_Cubic>,<at 1 1>]}}})
     Y=0
     Z=0
     W=0
    )
   ) {isSetParameters=true}
   /SetVariables(Particles.BG Glow Width:NiagaraFloat=&Multiply_Float(A=$Particles.RibbonWidth, B=1.12)) {isSetParameters=true}
  ]
  rendererClasses [&NiagaraRibbonRendererProperties]
  renderers [
   {rendererClass=&NiagaraRibbonRendererProperties}
   {rendererClass=&NiagaraRibbonRendererProperties, rendererIndex=1}
   {rendererClass=&NiagaraRibbonRendererProperties, rendererIndex=2}
   {rendererClass=&NiagaraRibbonRendererProperties, rendererIndex=3}
   {rendererClass=&NiagaraRibbonRendererProperties, rendererIndex=4}
  ]
 }
}
summary {
 systemName=NS_Pickaxe_AnodizedMetal_Trail
 emitters [
  {emitterName=Trails, rendererClasses [&NiagaraRibbonRendererProperties]}
  {emitterName=Symbols, rendererClasses [&NiagaraSpriteRendererProperties, &NiagaraRibbonRendererProperties]}
  {emitterName=TrailGalaxy, rendererClasses [&NiagaraRibbonRendererProperties]}
 ]
 userVariables [
  User.bIsFrontEnd {type=NiagaraBool}
  User.bIsFrontEndPreview {type=NiagaraBool}
  User.Debug {type=NiagaraBool}
  User.RibbonIDOffset {type=NiagaraInt32}
  User.SwingSectionIndex {type=NiagaraInt32}
  User.Variant {type=NiagaraFloat, defaultValue=0, description=""}
 ]
}
