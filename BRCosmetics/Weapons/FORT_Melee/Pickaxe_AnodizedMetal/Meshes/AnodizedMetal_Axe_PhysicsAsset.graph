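# Physics asset for the AnodizedMetal axe, a two-handed melee pickaxe cosmetic weapon mesh.
# A single capsule body on the Melee_TwoHanded_Handle bone approximates the weapon's collision.
# It has no constraints, so it simulates as one rigid piece.
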PhysicsAsset /BRCosmetics/Weapons/FORT_Melee/Pickaxe_AnodizedMetal/Meshes/AnodizedMetal_Axe_PhysicsAsset
bodies {Melee_TwoHanded_Handle [{shapeType=Capsule, length=55.4935, radius=6.2831, center=<-0.448 0.2302 33.3381>, rotation=<rot -0.3458 -90 0.0117>}]}
constraints []
